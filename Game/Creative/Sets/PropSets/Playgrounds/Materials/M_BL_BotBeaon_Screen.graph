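# Material for the Playgrounds prop set's bot beacon screen, drawing two animated glowing wave lines with scrolling scanlines.
# Panned, time-varying UVs drive sine functions that offset sphere masks on the V coordinate, forming procedural waves with pulsing amplitude.
# The tinted waves are added to a radial two-color background and multiplied by a panning gradient-texture scanline.
# The result is scaled and fed through MF_FX_EmissiveAdjustments, with parameters exposed for waves, colors and surface response.
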Material /Game/Creative/Sets/PropSets/Playgrounds/Materials/M_BL_BotBeaon_Screen
graph {
 properties {bUsedWithMeshParticles=true, bUsedWithNanite=true, pixelDepthOffsetMode=PDOM_Legacy, usedShadingModels=Default Lit}
 nodes [
  n1 = Sine(None=Time()) {period=1.87}
  n2 = Panner(
   Coordinate=LinearInterpolate(A=TextureCoordinate() {uTiling=1.5, vTiling=1.5}, B=TextureCoordinate() {uTiling=2.3, vTiling=2.3}, Alpha=@n1)
   Speed=ScalarParameter() {defaultValue=-3.585, parameterName=SpeedA}
  ) {bFractionalPart=true, speedX=-0.2}
  n3 = Sine(None=Multiply(A=ComponentMask(None=@n2) {r=true}, B=ScalarParameter() {defaultValue=0.9, parameterName=FrequencyA}))
  n4 = Multiply(A=ScalarParameter() {defaultValue=0.4, parameterName=SineStrengthA}, B=@n1)
  n5 = SphereMask(
   A=ComponentMask(None=TextureCoordinate()) {g=true}
   B=Add(A=Multiply(A=@n3, B=@n4), B=Constant() {r=0.5})
   Radius=ScalarParameter() {defaultValue=0.08, parameterName=SineRadius}
   Hardness=ScalarParameter() {parameterName=SineHardness}
  )
  n6 = Sine(None=Time()) {period=2.254}
  n7 = Panner(
   Coordinate=LinearInterpolate(A=TextureCoordinate() {uTiling=1.7, vTiling=1.7}, B=TextureCoordinate() {uTiling=2.5, vTiling=2.5}, Alpha=@n6)
   Speed=ScalarParameter() {defaultValue=-4.199, parameterName=SpeedB}
  ) {bFractionalPart=true, speedX=-0.2}
  n8 = Sine(None=Multiply(A=ComponentMask(None=@n7) {r=true}, B=ScalarParameter() {defaultValue=1.1, parameterName=FrequencyB}))
  n9 = Multiply(A=ScalarParameter() {defaultValue=0.2, parameterName=SineStrengthB}, B=@n6)
  n10 = SphereMask(
   A=ComponentMask(None=TextureCoordinate()) {g=true}
   B=Add(A=Multiply(A=@n8, B=@n9), B=Constant() {r=0.5})
   Radius=ScalarParameter() {defaultValue=0.08, parameterName=SineRadius}
   Hardness=ScalarParameter() {parameterName=SineHardness}
  )
  n11 = TextureCoordinate()
  n12 = Constant() {r=0.5}
  n13 = Constant() {r=0.8}
  n14 = Constant() {r=0.2}
  n15 = Add(
   A=Multiply(A=Add(A=@n5, B=@n10), B=VectorParameter() {parameterName=SineColorA, defaultValue=<rgba 0.2341 0.5165 0.9392 0>}.RGB)
   B=LinearInterpolate(
    A=VectorParameter() {parameterName=ScreenColorA, defaultValue=<rgba 0.0091 0.0287 0.1493 0>}.RGB
    B=VectorParameter() {parameterName=ScreenColorB, defaultValue=<rgba 0.0511 0.2886 0.9 0>}.RGB
    Alpha=SphereMask(A=@n11, B=@n12, Radius=@n13, Hardness=@n14)
   )
  )
  n16 = Multiply(A=TextureCoordinate(), B=ScalarParameter() {defaultValue=20, parameterName=ScanlineRepeat})
  n17 = Add(A=TextureSample(UVs=Panner(Coordinate=@n16) {bFractionalPart=true, speedY=-1.3}) {samplerType=SAMPLERTYPE_Masks, texture=/Game/Global/Textures/Default/Masks_and_Gradients/T_VertGradient_MASK_2}.R)
  MaterialFunctionCall(Emissive=Multiply(A=Multiply(A=@n15, B=@n17), B=ScalarParameter() {defaultValue=1, parameterName=Emissive})) {materialFunction=&MF_FX_EmissiveAdjustments}
  ScalarParameter() {defaultValue=0.2, parameterName=Metallic}
  ScalarParameter() {defaultValue=0.8, parameterName=Specular}
  ScalarParameter() {defaultValue=0.25, parameterName=Roughness}
 ]
}
