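# Lit, Nanite-enabled material for Creative Playgrounds wire props, showing segmented emissive pulses over a glossy metallic base.
# UVs scale by object radius and optionally reverse, then pan in quantized time steps to drive a ramped sweep, cut by repeating line masks.
# A U-axis edge mask shapes base color and glow.
# A static switch can source emissive color from custom primitive data, and attributes pass through MF_CreativeMoveTool.
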Material /Game/Creative/Sets/PropSets/Playgrounds/Materials/M_BL_Wires_Emissive
graph {
 properties {bUseMaterialAttributes=true, bUsedWithNanite=true, pixelDepthOffsetMode=PDOM_Legacy, usedShadingModels=Default Lit}
 nodes [
  n1 = ScalarParameter() {defaultValue=-1, parameterName=PanSpeedV}
  AppendVector(A=Constant(), B=@n1)
  n2 = VectorParameter() {parameterName=Base Color, defaultValue=<rgba 0.0071 0.1 0.0837 0>}
  n3 = Constant() {r=0.5}
  n4 = SphereMask(
   A=MaterialFunctionCall(Float2=TextureCoordinate()) {materialFunction=&BreakOutFloat2Components}.R
   B=@n3
   Radius=@n3
   Hardness=ScalarParameter() {defaultValue=0.2, parameterName=EdgeMaskHardness}
  )
  n5 = VectorParameter() {
   bUseCustomPrimitiveData=true
   desc="For when emissive absolutely needs to be done with primitive data on the mesh (likely Creative)"
   parameterName=EmissiveWithPrimitiveData
   channelNames {a=Emissive Multiplier, b="", g="", r=""}
   defaultValue=<rgba 0.2 3.3667 4 1>
  }
  n6 = ScalarParameter() {defaultValue=24, parameterName=LinesRepeat}
  n7 = ObjectRadius()
  n8 = TextureCoordinate()
  n9 = Multiply(
   A=Divide(
    A=MaterialFunctionCall(Float3=@n7) {materialFunction=&BreakOutFloat3Components}.B
    B=ScalarParameter() {defaultValue=32, parameterName=Divisor}
   )
   B=LinearInterpolate(A=@n8, B=Multiply(A=@n8) {constB=-1}, Alpha=ScalarParameter() {parameterName=ReverseDirection})
  )
  n10 = MaterialFunctionCall(Float2=@n9) {materialFunction=&BreakOutFloat2Components}
  n11 = SphereMask(
   A=Frac(None=Multiply(A=@n6, B=@n10.G))
   B=Constant()
   Radius=ScalarParameter() {defaultValue=0.4, parameterName=LineRadius}
   Hardness=ScalarParameter() {parameterName=LineHardness}
  )
  n12 = ScalarParameter() {defaultValue=12, parameterName=TimeSteps}
  n13 = Ceil(None=Multiply(A=Time(), B=@n12))
  n14 = Frac(None=Panner(Coordinate=@n9, Time=Divide(A=@n13, B=@n12), Speed=@n1) {bFractionalPart=true, speedX=1, speedY=1})
  n15 = ScalarParameter() {defaultValue=2, parameterName=UVExponent}
  n16 = SphereMask(
   A=MaterialFunctionCall(Float2=Power(Base=@n14, Exp=@n15)) {materialFunction=&BreakOutFloat2Components}.G
   B=ScalarParameter() {defaultValue=0.8, parameterName=Center}
   Radius=ScalarParameter() {defaultValue=0.8, parameterName=SweepRadius}
   Hardness=ScalarParameter() {defaultValue=0.05, parameterName=SweepHardness}
  )
  n17 = ScalarParameter() {defaultValue=5, parameterName=RampExponent}
  n18 = Multiply(
   A=StaticSwitchParameter(True=@n5, False=VectorParameter() {parameterName=Emissive, defaultValue=<rgba 0.2 3.3667 4 1>}.RGB) {parameterName=UseEmissiveWithPrimitiveData}
   B=Multiply(A=OneMinus(None=@n11), B=Power(Base=@n16, Exp=@n17))
  )
  MaterialFunctionCall(MA In=MakeMaterialAttributes!(
    BaseColor=Multiply(A=@n2.RGB, B=@n4)
    Metallic=ScalarParameter() {defaultValue=0.8, parameterName=Metallic}
    Specular=ScalarParameter() {defaultValue=0.66, parameterName=Specular}
    Roughness=ScalarParameter() {defaultValue=0.085, parameterName=Roughness}
    Anisotropy
    EmissiveColor=Multiply(A=@n18, B=@n4)
    Opacity
    OpacityMask
    Normal
    Tangent
    WorldPositionOffset
    SubsurfaceColor
    ClearCoat
    ClearCoatRoughness
    AmbientOcclusion
    Refraction
    CustomizedUVs_0
    CustomizedUVs_1
    CustomizedUVs_2
    CustomizedUVs_3
    CustomizedUVs_4
    CustomizedUVs_5
    CustomizedUVs_6
    CustomizedUVs_7
    PixelDepthOffset
    ShadingModel
    Displacement
   )) {materialFunction=&MF_CreativeMoveTool}
 ]
}
