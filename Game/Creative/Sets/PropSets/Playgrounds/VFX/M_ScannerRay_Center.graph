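# Additive unlit, two-sided material for the center of a scanner ray effect in the Playgrounds prop set, with color and angle from custom primitive data.
# Glow layers a bright tip, oscillating sweeps, panning noise and scrolling scanlines.
# World position offset opens the fan by Angle and wobbles secondary fans split by vertex color.
# Depth and near-camera fades shape opacity, while ScanRotAnim adds swinging rotation and an intersection highlight.
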Material /Game/Creative/Sets/PropSets/Playgrounds/VFX/M_ScannerRay_Center
graph {
 properties {
  bEnableResponsiveAA=true
  bUseMaterialAttributes=true
  blendMode=BLEND_Additive
  pixelDepthOffsetMode=PDOM_Legacy
  shadingModel=MSM_Unlit
  twoSided=true
  usedShadingModels=Unlit
 }
 nodes [
  Constant() {r=1}
  n1 = VectorParameter() {bUseCustomPrimitiveData=true, desc="10, 1.1, .295", parameterName=Color, primitiveDataIndex=8, defaultValue=<rgba 10 1.1039 0.2951 0>}
  n2 = ComponentMask(None=TextureCoordinate()) {g=true}
  n3 = ScalarParameter() {defaultValue=0.95, parameterName=TipCenter}
  n4 = ScalarParameter() {defaultValue=0.2, parameterName=TipRadius}
  n5 = ScalarParameter() {parameterName=TipHardness}
  n6 = Multiply(
   A=Power(Base=SphereMask(A=@n2, B=@n3, Radius=@n4, Hardness=@n5), Exp=ScalarParameter() {defaultValue=3, parameterName=TipExponent})
   B=ScalarParameter() {defaultValue=3, parameterName=TipScale}
  )
  n7 = Panner(Coordinate=TextureCoordinate()) {bFractionalPart=true, speedX=0.88}
  n8 = Time()
  n9 = Divide(A=Add(A=Sine(None=@n8) {period=0.3994}, B=Sine(None=@n8) {period=0.64}))
  n10 = Sign(None=@n9)
  n11 = ComponentMask(None=LinearInterpolate(A=@n7, B=OneMinus(None=@n7), Alpha=Saturate(None=@n10))) {r=true}
  n12 = ScalarParameter() {defaultValue=0.2, parameterName=RadiusA}
  n13 = Add(A=Abs(None=@n9)) {constB=0.5}
  n14 = Constant()
  n15 = Time()
  n16 = Divide(A=Add(A=Sine(None=@n15) {period=0.7604}, B=Sine(None=@n15) {period=0.2176}))
  n17 = ScalarParameter() {defaultValue=0.322, parameterName=RadiusB}
  n18 = Add(A=Abs(None=@n16)) {constB=0.5}
  n19 = Add(
   A=SphereMask(A=@n11, B=@n9, Radius=Multiply(A=@n12, B=@n13), Hardness=@n14)
   B=SphereMask(A=@n11, B=@n16, Radius=Multiply(A=@n17, B=@n18), Hardness=@n14)
  )
  n20 = ScalarParameter() {defaultValue=0.5, parameterName=SweepsOpacity}
  n21 = SphereMask(A=TextureCoordinate() {vTiling=0.4}, B=Constant2Vector() {g=-0.1, r=0.5}, Radius=Constant() {r=0.5}, Hardness=Constant())
  n22 = TextureCoordinate()
  n23 = VectorParameter() {parameterName=NoisePanScale, defaultValue=<rgba 0.6 0.05 0 0.18>}
  n24 = AppendVector(A=@n23.R, B=@n23.R)
  n25 = TextureSample(UVs=Panner(Coordinate=Multiply(A=@n22, B=@n24), Speed=AppendVector(A=@n23.B, B=@n23.B))) {texture=&AtmosphericCloudNoise01}
  n26 = Add(
   A=Power(Base=OneMinus(None=@n21), Exp=ScalarParameter() {defaultValue=4, parameterName=RadialExponent})
   B=Multiply(A=Subtract(A=@n25.R) {constB=0.5}, B=ScalarParameter() {defaultValue=0.5, parameterName=NoiseScale})
  )
  n27 = TextureCoordinate()
  n28 = AppendVector(A=Constant(), B=ScalarParameter() {defaultValue=0.1, parameterName=ScanlinePanSpeedV})
  n29 = Multiply(
   A=ComponentMask(None=Panner(Coordinate=@n27, Speed=@n28) {bFractionalPart=true, speedY=0.5}) {g=true}
   B=ScalarParameter() {defaultValue=110, parameterName=ScanlineSteps}
  )
  n30 = Power(
   Base=SphereMask(
    A=Frac(None=@n29)
    B=Constant() {r=0.5}
    Radius=ScalarParameter() {defaultValue=0.25, parameterName=ScanlinesRadius}
    Hardness=ScalarParameter() {parameterName=ScanlinesHardness}
   )
   Exp=ScalarParameter() {defaultValue=4, parameterName=ScanlineExponent}
  )
  n31 = Fmod(A=Multiply(A=Time() {period=6}, B=ScalarParameter() {defaultValue=3, parameterName=MaskSweepSpeed}), B=Constant() {r=1})
  n32 = SphereMask(
   A=ComponentMask(None=TextureCoordinate()) {g=true}
   B=Multiply(A=OneMinus(None=@n31)) {constB=6}
   Radius=ScalarParameter() {defaultValue=1, parameterName=MaskSweepRadius}
   Hardness=ScalarParameter() {parameterName=MaskSweepHardness}
  )
  n33 = Multiply(A=Multiply(A=Multiply(A=@n30, B=@n32), B=ScalarParameter() {defaultValue=3, parameterName=ScanlinesOpacity}), B=VertexColor().B)
  n34 = Multiply(
   A=Power(Base=@n2, Exp=ScalarParameter() {defaultValue=2.5, parameterName=LinearExponent}) {constExponent=3}
   B=Add(A=Multiply(A=@n19, B=@n20), B=Add(A=@n26, B=@n33))
  )
  n35 = Multiply(
   A=LinearInterpolate(A=@n1.RGB, B=Constant3Vector() {constant=<rgba 10 1.1 0.295 0>}, Alpha=ScalarParameter() {parameterName=DebugPreview})
   B=Add(A=Multiply(A=@n6, B=@n34), B=@n34)
  )
  n36 = DepthFade(FadeDistance=ScalarParameter() {defaultValue=20, parameterName=FadeDistance})
  n37 = OneMinus(None=@n36)
  n38 = Power(Base=@n37) {constExponent=8}
  n39 = Multiply(A=Multiply(A=@n35, B=Multiply(A=@n38) {constB=8}), B=ScalarParameter() {group=Scanner, parameterName=ScanRotAnim})
  n40 = ScalarParameter() {defaultValue=40, parameterName=NearCameraFade}
  n41 = ScalarParameter() {group=Scanner, parameterName=ScanRotAnim}
  n42 = ScalarParameter() {bUseCustomPrimitiveData=true, defaultValue=45, parameterName=Angle}
  n43 = ScalarParameter() {parameterName=DebugPreview}
  n44 = Multiply(
   A=Multiply(
    A=MaterialFunctionCall(Distance=@n40) {materialFunction=&Near_Camera_Fade}.Result
    B=LinearInterpolate(A=@n36, Alpha=@n41)
   )
   B=MaterialFunctionCall!(min, max=Constant() {r=5}, value=LinearInterpolate(A=@n42, Alpha=@n43) {constB=45}) {materialFunction=/Game/Effects/Fort_Effects/Materials/_Functions/InverseLerp_MF}.Clamped
  )
  n45 = ComponentMask(None=TextureCoordinate()) {r=true}
  n46 = Constant() {r=0.5}
  n47 = Constant()
  n48 = LinearInterpolate(
   A=Multiply(
    A=ScalarParameter() {bUseCustomPrimitiveData=true, defaultValue=0.5, desc=.5, parameterName=SecondaryFanOpacity, primitiveDataIndex=13}
    B=SphereMask(A=@n45, B=@n46, Radius=@n46, Hardness=@n47)
   )
   Alpha=VertexColor().B
  ) {constA=1}
  n49 = Transform(None=Constant3Vector() {constant=<rgba 0 1 0 0>}) {transformSourceType=TRANSFORMSOURCE_Local}
  n50 = Multiply(A=Time(), B=ScalarParameter() {bUseCustomPrimitiveData=true, defaultValue=1, parameterName=SecondaryMovementSpeed, primitiveDataIndex=12})
  n51 = Sine(None=@n50) {period=0.4154}
  n52 = Sine(None=@n50) {period=0.2767}
  n53 = LinearInterpolate(A=@n51, B=@n52, Alpha=VertexColor().R)
  n54 = ScalarParameter() {defaultValue=0.013, parameterName=ScaleSecondaryAngle}
  n55 = Constant3Vector() {constant=<rgba 0 0 1 0>}
  n56 = ScalarParameter() {bUseCustomPrimitiveData=true, defaultValue=45, parameterName=Angle}
  n57 = ScalarParameter() {parameterName=DebugPreview}
  n58 = Add(A=@n51, B=@n52)
  n59 = ScalarParameter() {defaultValue=0.25, parameterName=AngleWiggleScale}
  n60 = Divide(A=Add(A=LinearInterpolate(A=@n56, Alpha=@n57) {constB=45}, B=Multiply(A=@n58, B=@n59))) {constB=720}
  n61 = TextureCoordinate()
  n62 = Sign(None=Subtract(A=ComponentMask(None=@n61) {r=true}) {constB=0.5})
  n63 = RotateAboutAxis(
   NormalizedRotationAxis=Normalize(VectorInput=Transform(None=@n55) {transformSourceType=TRANSFORMSOURCE_Local})
   RotationAngle=LinearInterpolate(A=@n60, B=Multiply(A=@n60) {constB=-1}, Alpha=Saturate(None=@n62))
   PivotPoint=MaterialFunctionCall() {materialFunction=&ObjectPivotPoint}.Object Pivot Location
   Position=WorldPosition().XYZ
  )
  n64 = Add(
   A=RotateAboutAxis(
    NormalizedRotationAxis=Normalize(VectorInput=@n49)
    RotationAngle=Multiply(A=@n53, B=@n54)
    PivotPoint=MaterialFunctionCall() {materialFunction=&ObjectPivotPoint}.Object Pivot Location
    Position=WorldPosition().XYZ
   )
   B=@n63
  )
  n65 = VertexColor()
  n66 = ScalarParameter() {group=Scanner, parameterName=ScanRotAnim}
  n67 = Constant3Vector() {constant=<rgba 0 1 0 0>}
  n68 = LinearInterpolate(Alpha=MaterialFunctionCall() {materialFunction=&DebugTimeSine}.Result) {constA=-0.1, constB=0.1}
  n69 = ScalarParameter() {defaultValue=0.2, group=Scanner, parameterName=ScanRotAmount}
  n70 = Clamp(None=Multiply(
    A=MaterialFunctionCall() {materialFunction=/Engine/Functions/Engine_MaterialFunctions02/WorldPositionOffset/ObjectScale}.Scale X
    B=ScalarParameter() {defaultValue=0.1, group=Scanner, parameterName=ScanRotDistanceScale}
   ) {constB=0.25}) {minDefault=-1}
  n71 = RotateAboutAxis(
   NormalizedRotationAxis=Normalize(VectorInput=Transform(None=@n67) {transformSourceType=TRANSFORMSOURCE_Local})
   RotationAngle=Multiply(A=Multiply(A=@n68, B=@n69), B=LinearInterpolate(Alpha=@n70) {constA=1, constB=0})
   PivotPoint=MaterialFunctionCall() {materialFunction=&ObjectPivotPoint}.Object Pivot Location
   Position=WorldPosition().XYZ
  )
  MakeMaterialAttributes!(
   BaseColor
   Metallic
   Specular
   Roughness
   Anisotropy
   EmissiveColor=MaterialFunctionCall(Emissive=Add(A=@n35, B=@n39)) {materialFunction=&MF_FX_EmissiveAdjustments}.Result
   Opacity=Add(A=Multiply(A=@n44, B=@n48), B=Multiply(A=@n37, B=@n41))
   OpacityMask
   Normal
   Tangent
   WorldPositionOffset=Add(A=LinearInterpolate(A=@n64, B=@n63, Alpha=@n65.B), B=Multiply(A=@n66, B=@n71)) {constB=0}
   SubsurfaceColor
   ClearCoat
   ClearCoatRoughness
   AmbientOcclusion
   Refraction
   CustomizedUVs_0
   CustomizedUVs_1
   CustomizedUVs_2
   CustomizedUVs_3
   CustomizedUVs_4
   CustomizedUVs_5
   CustomizedUVs_6
   CustomizedUVs_7
   PixelDepthOffset
   ShadingModel
   Displacement
  )
 ]
}
