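# Additive unlit glow material for the scanner ray in the Creative Playgrounds prop set, drawn as a camera-facing flare.
# A grayscale flare texture is raised to an exponent that two summed sine waves modulate for flicker.
# It is tinted by a custom-primitive-data Color, with a DebugPreview blend to fixed orange-red, then passed through MF_FX_EmissiveAdjustments.
# World position offset aligns and offsets the mesh toward the camera, while DepthFade drives opacity.
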Material /Game/Creative/Sets/PropSets/Playgrounds/VFX/M_ScannerRay_Glow
graph {
 properties {
  bEnableResponsiveAA=true
  bUseMaterialAttributes=true
  blendMode=BLEND_Additive
  pixelDepthOffsetMode=PDOM_Legacy
  shadingModel=MSM_Unlit
  usedShadingModels=Unlit
 }
 nodes [
  n1 = VectorParameter() {
   bUseCustomPrimitiveData=true
   desc="20.0, 1.022216, 0.36781"
   parameterName=Color
   primitiveDataIndex=4
   defaultValue=<rgba 2.8365 4 0.8749 0>
  }
  n2 = LinearInterpolate(A=@n1.RGB, B=Constant3Vector() {constant=<rgba 20 1 0.3678 0>}, Alpha=ScalarParameter() {parameterName=DebugPreview})
  n3 = Time()
  n4 = Add(A=Sine(None=@n3) {period=0.231}, B=Sine(None=@n3) {period=0.159})
  n5 = ScalarParameter() {defaultValue=0.4, parameterName=FlickerExponentBiasScale}
  n6 = Power(
   Base=TextureSampleParameter2D() {parameterName=Texture Param, samplerType=SAMPLERTYPE_LinearGrayscale, texture=/Game/FrontEnd/Store/Cards/Assets/T_5Flare_GS}.RGB
   Exp=Add(A=ScalarParameter() {defaultValue=2.8, parameterName=Exponent}, B=Multiply(A=@n4, B=@n5))
  )
  n7 = ScalarParameter() {defaultValue=10, parameterName=Camera Offset}
  n8 = Normalize(VectorInput=Transform(None=Constant3Vector() {constant=<rgba 0 0 1 0>}) {transformSourceType=TRANSFORMSOURCE_Local})
  n9 = Normalize(VectorInput=Transform(None=Constant3Vector() {constant=<rgba 0 1 0 0>}) {transformSourceType=TRANSFORMSOURCE_Local})
  n10 = Normalize(VectorInput=Transform(None=Constant3Vector() {constant=<rgba 1 0 0 0>}) {transformSourceType=TRANSFORMSOURCE_Local})
  MakeMaterialAttributes!(
   BaseColor
   Metallic
   Specular
   Roughness
   Anisotropy
   EmissiveColor=MaterialFunctionCall(Emissive=Multiply(A=@n2, B=@n6) {constA=1}) {materialFunction=&MF_FX_EmissiveAdjustments}.Result
   Opacity=DepthFade() {fadeDistanceDefault=50}
   OpacityMask
   Normal
   Tangent
   WorldPositionOffset=Add(
    A=MaterialFunctionCall(Offset Amount=@n7) {materialFunction=/Engine/Functions/Engine_MaterialFunctions02/WorldPositionOffset/CameraOffset}.Camera Offset
    B=MaterialFunctionCall!(WorldPosition, PivotLocation, WS Normals, Custom Object Basis 1=@n8, Custom Object Basis 2=@n9, Custom Object Basis 3=@n10) {materialFunction=/Engine/Functions/Engine_MaterialFunctions02/WorldPositionOffset/AlignMeshToTheCamera}.World Position Offset
   )
   SubsurfaceColor
   ClearCoat
   ClearCoatRoughness
   AmbientOcclusion
   Refraction
   CustomizedUVs_0
   CustomizedUVs_1
   CustomizedUVs_2
   CustomizedUVs_3
   CustomizedUVs_4
   CustomizedUVs_5
   CustomizedUVs_6
   CustomizedUVs_7
   PixelDepthOffset
   ShadingModel
   Displacement
  )
 ]
}
