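# Additive, unlit, two-sided material for the glowing lines of a scanner ray prop effect, with color, angle, widths, falloff and wiggle speed driven by custom primitive data.
# Emissive combines UV sphere-mask gradients with two panning, scale-tiled noise layers and a low-quality visibility fix.
# Depth and near-camera fades set opacity.
# World position offset thickens lines via SplineThicken, spreads them by vertex-color-mirrored rotation with sine wiggle, and adds optional scan rotation.
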Material /Game/Creative/Sets/PropSets/Playgrounds/VFX/M_ScannerRay_Lines
graph {
 properties {
  bEnableResponsiveAA=true
  bUseMaterialAttributes=true
  blendMode=BLEND_Additive
  pixelDepthOffsetMode=PDOM_Legacy
  shadingModel=MSM_Unlit
  twoSided=true
  usedShadingModels=Unlit
 }
 nodes [
  n1 = TextureCoordinate() {uTiling=0}
  n2 = Constant()
  n3 = MaterialFunctionCall() {materialFunction=&ObjectScale}
  n4 = ScalarParameter() {defaultValue=0.5, parameterName=BreakupTilingScale_1}
  n5 = Multiply(A=@n1, B=AppendVector(A=@n2, B=Multiply(A=@n3.Scale X, B=@n4)))
  n6 = SmoothStep(Value=TextureSample(UVs=Panner(Coordinate=@n5) {speedX=0.25, speedY=0.5}) {texture=/Game/Global/Textures/Default/Noise/3xPerlin_M}.R) {constMax=0.33}
  n7 = ScalarParameter() {defaultValue=0.7, parameterName=BreakupTilingScale_2}
  n8 = Multiply(A=@n1, B=AppendVector(A=@n2, B=Multiply(A=@n3.Scale X, B=@n7)))
  n9 = SmoothStep(Value=TextureSample(UVs=Panner(Coordinate=@n8) {speedX=0.15, speedY=0.33}) {texture=&3xPerlin_M}.G) {constMax=0.5}
  n10 = LinearInterpolate(
   A=Multiply(A=Add(A=@n6) {constB=0.1}, B=Add(A=@n9) {constB=0.3})
   Alpha=MaterialFunctionCall(
    Fade Length=ScalarParameter() {defaultValue=1024, parameterName=CameraFadeDepthLenght}
    Fade Offset=ScalarParameter() {defaultValue=512, parameterName=CameraFadeOffset}
   ) {materialFunction=&CameraDepthFade}.Result
  ) {constA=1}
  n11 = TextureCoordinate()
  n12 = ComponentMask(None=@n11) {g=true}
  n13 = Constant() {r=1}
  n14 = ComponentMask(None=@n11) {r=true}
  n15 = Constant() {r=0.5}
  n16 = Constant()
  n17 = ScalarParameter() {defaultValue=15, parameterName=GradientExponent}
  n18 = Constant() {r=10}
  n19 = QualitySwitch(Default=@n17, Low=@n18, Medium=@n18)
  n20 = ScalarParameter() {parameterName=VisibilityFixOnLow}
  n21 = Multiply(
   A=SphereMask(
    A=@n12
    B=@n13
    Radius=@n13
    Hardness=ScalarParameter() {bUseCustomPrimitiveData=true, defaultValue=0.25, desc=.25, parameterName=FalloffHardness, primitiveDataIndex=3}
   )
   B=Power(Base=SphereMask(A=@n14, B=@n15, Radius=@n15, Hardness=@n16), Exp=LinearInterpolate(A=@n17, B=@n19, Alpha=@n20))
  )
  n22 = Multiply(A=@n21, B=Constant() {r=15}) {constB=200}
  n23 = VectorParameter() {
   bUseCustomPrimitiveData=true
   desc="20.0, 1.022216, 0.36781"
   parameterName=Color
   primitiveDataIndex=4
   defaultValue=<rgba 20 1.0222 0.3678 0>
  }
  n24 = Multiply(
   A=LinearInterpolate(A=@n21, B=QualitySwitch(Default=@n21, Low=@n22, Medium=@n22), Alpha=ScalarParameter() {parameterName=VisibilityFixOnLow})
   B=LinearInterpolate(A=@n23.RGB, B=Constant3Vector() {constant=<rgba 20 1 0.3 0>}, Alpha=ScalarParameter() {parameterName=DebugPreview})
  )
  n25 = DepthFade(FadeDistance=ScalarParameter() {defaultValue=20, parameterName=FadeDistance})
  n26 = MaterialFunctionCall(Distance=ScalarParameter() {defaultValue=40, parameterName=NearCameraFade}) {materialFunction=&Near_Camera_Fade}
  n27 = ScalarParameter() {bUseCustomPrimitiveData=true, defaultValue=45, parameterName=Angle}
  n28 = ScalarParameter() {parameterName=DebugPreview}
  n29 = Multiply(A=Time(), B=ScalarParameter() {bUseCustomPrimitiveData=true, defaultValue=1, parameterName=SecondaryMovementSpeed, primitiveDataIndex=12})
  n30 = Multiply(
   A=Add(A=Sine(None=@n29) {period=0.4154}, B=Sine(None=@n29) {period=0.2767})
   B=ScalarParameter() {defaultValue=0.25, parameterName=AngleWiggleScale}
  )
  n31 = MaterialFunctionCall!(
   min
   max=Constant() {r=5}
   value=LinearInterpolate(
    A=ScalarParameter() {bUseCustomPrimitiveData=true, defaultValue=45, parameterName=Angle}
    Alpha=ScalarParameter() {parameterName=DebugPreview}
   ) {constB=45}
  ) {materialFunction=/Game/Effects/Fort_Effects/Materials/_Functions/InverseLerp_MF}
  n32 = Divide(A=Add(A=LinearInterpolate(A=@n27, Alpha=@n28) {constB=45}, B=Multiply(A=@n30, B=@n31.Clamped))) {constB=720}
  n33 = RotateAboutAxis(
   NormalizedRotationAxis=Normalize(VectorInput=Constant3Vector() {constant=<rgba 0 0 1 0>})
   RotationAngle=LinearInterpolate(A=@n32, B=Multiply(A=@n32) {constB=-1}, Alpha=VertexColor().R)
   PivotPoint=MaterialFunctionCall() {materialFunction=&ObjectPivotPoint}.Object Pivot Location
   Position=WorldPosition().XYZ
  )
  n34 = ScalarParameter() {bUseCustomPrimitiveData=true, defaultValue=4, desc=4, parameterName=WidthEnd, primitiveDataIndex=1}
  n35 = ScalarParameter() {bUseCustomPrimitiveData=true, defaultValue=18, desc=18, parameterName=WidthBase, primitiveDataIndex=2}
  n36 = ScalarParameter() {parameterName=DebugPreview}
  n37 = MaterialFunctionCall(
   WidthBase=LinearInterpolate(A=LinearInterpolate(A=@n34, B=@n35, Alpha=@n12), B=LinearInterpolate(Alpha=@n12) {constA=4, constB=18}, Alpha=@n36) {constB=45}
   WidthTip=LinearInterpolate(A=@n34, Alpha=@n36) {constB=4}
  ) {materialFunction=&SplineThicken}
  n38 = ScalarParameter() {group=Scanner, parameterName=ScanRotAnim}
  n39 = Constant3Vector() {constant=<rgba 0 1 0 0>}
  n40 = LinearInterpolate(Alpha=MaterialFunctionCall() {materialFunction=&DebugTimeSine}.Result) {constA=-0.1, constB=0.1}
  n41 = ScalarParameter() {defaultValue=0.2, group=Scanner, parameterName=ScanRotAmount}
  n42 = Clamp(None=Multiply(
    A=MaterialFunctionCall() {materialFunction=/Engine/Functions/Engine_MaterialFunctions02/WorldPositionOffset/ObjectScale}.Scale X
    B=ScalarParameter() {defaultValue=0.1, group=Scanner, parameterName=ScanRotDistanceScale}
   ) {constB=0.25}) {minDefault=-1}
  n43 = RotateAboutAxis(
   NormalizedRotationAxis=Normalize(VectorInput=Transform(None=@n39) {transformSourceType=TRANSFORMSOURCE_Local})
   RotationAngle=Multiply(A=Multiply(A=@n40, B=@n41), B=LinearInterpolate(Alpha=@n42) {constA=1, constB=0})
   PivotPoint=MaterialFunctionCall() {materialFunction=&ObjectPivotPoint}.Object Pivot Location
   Position=WorldPosition().XYZ
  )
  MakeMaterialAttributes!(
   BaseColor
   Metallic
   Specular
   Roughness
   Anisotropy
   EmissiveColor=MaterialFunctionCall(Emissive=Multiply(A=@n10, B=@n24)) {materialFunction=&MF_FX_EmissiveAdjustments}.Result
   Opacity=Multiply(A=Multiply(A=@n25, B=@n26.Result), B=ScalarParameter() {defaultValue=1, parameterName=LinesOpacity})
   OpacityMask
   Normal
   Tangent
   WorldPositionOffset=Add(A=Add(A=@n33, B=@n37.WorldPositionOffset), B=Multiply(A=@n38, B=@n43))
   SubsurfaceColor
   ClearCoat
   ClearCoatRoughness
   AmbientOcclusion
   Refraction
   CustomizedUVs_0
   CustomizedUVs_1
   CustomizedUVs_2
   CustomizedUVs_3
   CustomizedUVs_4
   CustomizedUVs_5
   CustomizedUVs_6
   CustomizedUVs_7
   PixelDepthOffset
   ShadingModel
   Displacement
  )
  Transform(None=@n33) {transformSourceType=TRANSFORMSOURCE_Local}
 ]
}
